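# Translucent unlit mesh-particle material for Taker creature suck tendrils, driven by dynamic parameters.
# SplineThicken sets strand width, swelled by a panning throb mask.
# World position offset adds either sine wave plus tip-weighted bend or Jiggle_WPO, chosen by static switch; the bend reads only BendDirectionX for all three axes.
# Particle color tints sun-direction shading plus a pannable V gradient, which also drives opacity with optional depth fade.
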Material /Game/Effects/Fort_Effects/Materials/Creatures/Taker/M_SuckTendrils
graph {
 properties {
  bEnableResponsiveAA=true
  bUsedWithMeshParticles=true
  bUsedWithNiagaraMeshParticles=true
  blendMode=BLEND_Translucent
  pixelDepthOffsetMode=PDOM_Legacy
  shadingModel=MSM_Unlit
  twoSided=true
  usedShadingModels=Unlit
 }
 nodes [
  n1 = CollectionParameter() {
   collection=&FortniteMaterialParameters
   parameterName=SunlightDirectionalVector
  }
  n2 = DynamicParameter() {defaultValue=<rgba 0 0 0 1>, paramNames [WidthBase,WidthTip,PanMask,PulseSpeed]}
  n3 = ScalarParameter() {defaultValue=3, parameterName=Thickness}
  n4 = Time()
  n5 = TextureSample(UVs=Panner(Time=Multiply(A=@n2.PulseSpeed, B=@n4)) {speedY=-1.5}) {texture=/Game/Effects/Fort_Effects/Textures/Taker/T_ThrobMask}
  n6 = ComponentMask(None=TextureCoordinate()) {g=true}
  n7 = StaticBool() {value=true}
  n8 = MaterialFunctionCall(
   WidthBase=LinearInterpolate(A=@n2.WidthBase, B=Multiply(A=@n2.WidthBase, B=@n3), Alpha=@n5.R)
   WidthTip=LinearInterpolate(A=@n2.WidthTip, B=Multiply(A=@n2.WidthTip, B=@n3), Alpha=@n5.R)
   UVs for Thickness=Clamp(None=LinearInterpolate(Alpha=@n6) {constA=-0.5})
   Expand U or V UV Channel=@n7
   AngleCorrectedNormals=@n7
  ) {materialFunction=&SplineThicken}
  n9 = DotProduct(A=ComponentMask(None=@n1) {b=true, g=true, r=true}, B=@n8.Normal)
  n10 = TextureCoordinate()
  n11 = LinearInterpolate(Alpha=ComponentMask(None=Add(A=@n10, B=@n2.PanMask) {constB=0, desc=0 to -1 to pan Mask}) {g=true}) {constA=-0.5}
  n12 = ParticleColor()
  MaterialFunctionCall(Emissive=Multiply(A=Add(A=@n9, B=@n11), B=@n12.RGB)) {materialFunction=&MF_FX_EmissiveAdjustments}
  n13 = OneMinus(None=ComponentMask(None=TextureCoordinate()) {g=true})
  n14 = DynamicParameter() {parameterIndex=2, paramNames [SinScale, SinSpeed, SinIntensity, BendDirectionZ]}
  n15 = Time()
  n16 = Sine(None=Add(A=Divide(A=@n13, B=@n14.SinScale), B=Multiply(A=@n15, B=@n14.SinSpeed)))
  n17 = Multiply(A=VectorParameter() {parameterName=WorldDirection, defaultValue=<rgba 1 0 0 1>}.RGB, B=@n14.SinIntensity)
  n18 = OneMinus(None=ComponentMask(None=TextureCoordinate()) {g=true})
  n19 = Sine(None=Multiply(A=@n18) {constB=0.25})
  n20 = ScalarParameter() {defaultValue=4, parameterName=Exp}
  n21 = DynamicParameter() {parameterIndex=1, defaultValue=<rgba 0 0 0 1>, paramNames [BendStrength, BendDirectionX, BendDirectionY, BendDirectionZ]}
  n22 = MaterialFunctionCall(R=@n21.BendDirectionX, G=@n21.BendDirectionX, B=@n21.BendDirectionX) {materialFunction=&AppendMany}
  n23 = Add(
   A=Multiply(A=Multiply(A=@n16, B=@n17), B=@n18)
   B=Multiply(A=Power(Base=@n19, Exp=@n20) {constExponent=4}, B=Multiply(A=@n21.BendStrength, B=@n22.RGB))
  )
  n24 = Transform(None=MaterialFunctionCall!(
    Speed=ScalarParameter() {parameterName=WPO_Speed}
    SmallWaveScle=ScalarParameter() {defaultValue=1, parameterName=WPO_SmWaveScale}
    WaveScale=ScalarParameter() {defaultValue=30, parameterName=WPO_WaveScale}
   ) {materialFunction=/Game/Effects/Fort_Effects/Materials/_Functions/Jiggle_WPO}.Result)
  Reroute(None=Add(A=@n8.WorldPositionOffset, B=StaticSwitchParameter(True=@n23, False=@n24) {parameterName=UseOrganicWave}))
  n25 = Multiply(A=@n11, B=@n12.A)
  n26 = ScalarParameter() {defaultValue=10, parameterName=FadeDistance}
  StaticSwitchParameter(
   True=ShadingPathSwitch(Default=DepthFade(Opacity=@n25, FadeDistance=@n26), Mobile=Constant() {r=1}) {desc=Migrated from FeatureLevelSwitch}
   False=@n25
  ) {parameterName=UseDepthFade?}
  VectorParameter() {parameterName=BendDirection, defaultValue=<rgba 1 0 0 1>}
 ]
}
